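# Physics asset for the frontend Front Flap mesh of the male medium Skull Punk body, providing secondary motion setup for dangling parts.
# Limited-angle constraints chain two front-flap capsules and four right-strap capsules from their MSTR roots, with tight twist limits.
# Pelvis, thigh and straps_driven_col capsules form body volumes; their joints and the MSTR root joints swing and twist freely.
# Back flap and left strap have only tiny root capsules, without chains.
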PhysicsAsset /BRCosmetics/Characters/Player/Male/Medium/Bodies/M_MED_Skull_Punk/Meshes/M_MED_Skull_Punk_Front_Flap_frontend_Physics
bodies {
 dyn_back_flap_MSTR [{shapeType=Capsule, length=0.505, radius=0.505, shapeName=dyn_back_flap_MSTR_capsule, rotation=<rot 0 -90 0>}]
 dyn_flap_front_01 [{shapeType=Capsule, length=19.1695, radius=4.0629, shapeName=dyn_flap_front_01_capsule, center=<7 0 4>, rotation=<rot 0 0 94.9989>}]
 dyn_flap_front_02 [{shapeType=Capsule, length=14.0477, radius=5.1848, shapeName=dyn_flap_front_02_capsule, center=<7 0 4>, rotation=<rot 0 0 90.0001>}]
 dyn_flap_front_MSTR [{shapeType=Capsule, length=0.505, radius=0.505, shapeName=dyn_flap_front_MSTR_capsule, rotation=<rot 0 -90 0>}]
 dyn_strap_l_MSTR [{shapeType=Capsule, length=0.505, radius=0.505, shapeName=dyn_strap_l_MSTR_capsule, rotation=<rot 0 -90 0>}]
 dyn_strap_r_01 [{
   shapeType=Capsule
   length=3.7538
   radius=1.8547
   shapeName=dyn_strap_r_01_capsule
   center=<3.7397 -0.1005 -2.0164>
   rotation=<rot 0 -104.1394 -98.1173>
  }]
 dyn_strap_r_02 [{
   shapeType=Capsule
   length=3.6134
   radius=1.9951
   shapeName=dyn_strap_r_02_capsule
   center=<3.7046 -0.0519 -2.2698>
   rotation=<rot 0 90.9544 -94.7971>
  }]
 dyn_strap_r_03 [{
   shapeType=Capsule
   length=5.0986
   radius=1.7013
   shapeName=dyn_strap_r_03_capsule
   center=<4.2755 0.0511 -1.8001>
   rotation=<rot 0 -89.4263 -95.7932>
  }]
 dyn_strap_r_04 [{shapeType=Capsule, length=2.2594, radius=1.7557, shapeName=dyn_strap_r_04_capsule, center=<1.2524 0 0>, rotation=<rot 0 0 90>}]
 dyn_strap_r_MSTR [{shapeType=Capsule, length=0.505, radius=0.505, shapeName=dyn_strap_r_MSTR_capsule, rotation=<rot 0 -90 0>}]
 dyn_straps_driven_col [{
   shapeType=Capsule
   length=10.5
   radius=18.255
   shapeName=dyn_straps_driven_col_capsule
   center=<4.7758 0.0002 -18.2669>
   rotation=<rot 75.8136 179.9999 179.9999>
  }]
 pelvis [
  {shapeType=Capsule, length=9.0699, radius=8.9868, center=<0 -2.9946 1.4988>}
  {shapeType=Capsule, length=9.4373, radius=9.7821, shapeName=pelvis_capsule, center=<-6.9516 0.7229 -3.6819>}
 ]
 thigh_l [{
   shapeType=Capsule
   length=16.8679
   radius=8.6002
   shapeName=thigh_l_capsule
   center=<-15.9432 1.2775 1.8553>
   rotation=<rot 90 180 -175.3141>
  }]
 thigh_r [{shapeType=Capsule, length=16.7745, radius=8.6936, shapeName=thigh_r_capsule, center=<14.5658 0.13 -0.8394>, rotation=<rot 90 35.2646 37>}]
}
constraints [
 {bone1Name=dyn_flap_front_01, bone2Name=dyn_flap_front_MSTR, swing1LimitDegrees=15, swing2LimitDegrees=45, twistLimitDegrees=5}
 {
  bone1Name=dyn_flap_front_MSTR
  bone2Name=pelvis
  swing1LimitDegrees=45
  swing1Motion=Free
  swing2LimitDegrees=45
  swing2Motion=Free
  twistLimitDegrees=45
  twistMotion=Free
 }
 {bone1Name=dyn_flap_front_02, bone2Name=dyn_flap_front_01, swing1LimitDegrees=15, swing2LimitDegrees=45, twistLimitDegrees=5}
 {
  bone1Name=dyn_back_flap_MSTR
  bone2Name=pelvis
  swing1LimitDegrees=45
  swing1Motion=Free
  swing2LimitDegrees=45
  swing2Motion=Free
  twistLimitDegrees=45
  twistMotion=Free
 }
 {
  bone1Name=thigh_r
  bone2Name=pelvis
  swing1LimitDegrees=45
  swing1Motion=Free
  swing2LimitDegrees=45
  swing2Motion=Free
  twistLimitDegrees=45
  twistMotion=Free
 }
 {
  bone1Name=dyn_strap_l_MSTR
  bone2Name=thigh_l
  swing1LimitDegrees=45
  swing1Motion=Free
  swing2LimitDegrees=45
  swing2Motion=Free
  twistLimitDegrees=45
  twistMotion=Free
 }
 {
  bone1Name=dyn_strap_r_MSTR
  bone2Name=pelvis
  swing1LimitDegrees=45
  swing1Motion=Free
  swing2LimitDegrees=45
  swing2Motion=Free
  twistLimitDegrees=45
  twistMotion=Free
 }
 {bone1Name=dyn_strap_r_01, bone2Name=dyn_strap_r_MSTR, swing1LimitDegrees=45, swing2LimitDegrees=45, twistLimitDegrees=3}
 {bone1Name=dyn_strap_r_02, bone2Name=dyn_strap_r_01, swing1LimitDegrees=45, swing2LimitDegrees=45, twistLimitDegrees=3}
 {bone1Name=dyn_strap_r_03, bone2Name=dyn_strap_r_02, swing1LimitDegrees=45, swing2LimitDegrees=45, twistLimitDegrees=3}
 {bone1Name=dyn_strap_r_04, bone2Name=dyn_strap_r_03, swing1LimitDegrees=45, swing2LimitDegrees=45, twistLimitDegrees=3}
 {
  bone1Name=thigh_l
  bone2Name=pelvis
  swing1LimitDegrees=45
  swing1Motion=Free
  swing2LimitDegrees=45
  swing2Motion=Free
  twistLimitDegrees=45
  twistMotion=Free
 }
 {
  bone1Name=dyn_straps_driven_col
  bone2Name=pelvis
  swing1LimitDegrees=45
  swing1Motion=Free
  swing2LimitDegrees=45
  swing2Motion=Free
  twistLimitDegrees=45
  twistMotion=Free
 }
]
