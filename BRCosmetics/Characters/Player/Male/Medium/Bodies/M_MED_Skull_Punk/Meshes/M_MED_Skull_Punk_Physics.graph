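# Physics asset for the Skull Punk male medium cosmetic body, covering its dangling back flap and left and right straps.
# Box and capsule chains with limited swing and narrow twist hang from master bones with free angular motion.
# The left strap master joins the left thigh; the right strap, back flap and chainless front flap masters join the pelvis.
# Pelvis, thigh and large pelvis-linked dyn_straps_driven_col shapes provide body collision volumes.
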PhysicsAsset /BRCosmetics/Characters/Player/Male/Medium/Bodies/M_MED_Skull_Punk/Meshes/M_MED_Skull_Punk_Physics
bodies {
 dyn_back_flap_01 [{shapeType=Box, extentX=10.2501, extentY=3.8865, extentZ=21.0997, shapeName=dyn_back_flap_01_box, center=<4 0 0>, rotation=<rot 0 0 90>}]
 dyn_back_flap_02 [{shapeType=Capsule, length=4.682, radius=6.728, shapeName=dyn_back_flap_02_capsule, center=<5.3631 0 4>, rotation=<rot 0 0 89.9999>}]
 dyn_back_flap_MSTR [{shapeType=Capsule, length=0.505, radius=0.505, shapeName=dyn_back_flap_MSTR_capsule, rotation=<rot 0 -90 0>}]
 dyn_flap_front_MSTR [{shapeType=Capsule, length=0.505, radius=0.505, shapeName=dyn_flap_front_MSTR_capsule, rotation=<rot 0 -90 0>}]
 dyn_strap_l_01 [{shapeType=Capsule, length=4.5134, radius=1.3837, shapeName=dyn_strap_l_01_capsule, center=<4 0 1>, rotation=<rot 90 0 0>}]
 dyn_strap_l_02 [{shapeType=Capsule, length=4.8869, radius=1.8765, shapeName=dyn_strap_l_02_capsule, center=<4 0 1>, rotation=<rot 90 0 0>}]
 dyn_strap_l_03 [{shapeType=Capsule, length=2.2947, radius=1.9758, shapeName=dyn_strap_l_03_capsule, center=<3.5 0 1>, rotation=<rot 90 0 0>}]
 dyn_strap_l_04 [{shapeType=Capsule, length=2.2584, radius=1.7611, shapeName=dyn_strap_l_04_capsule, center=<1.3377 0 0>, rotation=<rot 0 179.8973 -80.6402>}]
 dyn_strap_l_MSTR [{shapeType=Capsule, length=0.505, radius=0.505, shapeName=dyn_strap_l_MSTR_capsule, rotation=<rot 0 -90 0>}]
 dyn_strap_r_01 [{shapeType=Box, extentX=3.5, extentY=3, extentZ=1.01, shapeName=dyn_strap_r_01_box, center=<4 0 0>}]
 dyn_strap_r_02 [{shapeType=Box, extentX=6.5, extentY=3, extentZ=1, shapeName=dyn_strap_r_02_box, center=<3 0 0>}]
 dyn_strap_r_03 [{shapeType=Box, extentX=6, extentY=3, extentZ=2, shapeName=dyn_strap_r_03_box, center=<3 0 0>}]
 dyn_strap_r_04 [{shapeType=Capsule, length=2.3499, radius=2.1521, shapeName=dyn_strap_r_04_capsule, center=<1.042 0 0>, rotation=<rot 0 0 89.9999>}]
 dyn_strap_r_MSTR [{shapeType=Capsule, length=0.505, radius=0.505, shapeName=dyn_strap_r_MSTR_capsule, rotation=<rot 0 -90 0>}]
 dyn_straps_driven_col [{
   shapeType=Capsule
   length=10.5
   radius=18.255
   shapeName=dyn_straps_driven_col_capsule
   center=<3.6532 0.0002 -19.0709>
   rotation=<rot 85.8137 -180 -180>
  }]
 pelvis [{shapeType=Capsule, length=7.3841, radius=8.3302, center=<0 -0.2135 1.4988>}]
 thigh_l [
  {
   shapeType=Capsule
   length=16.8679
   radius=8.6002
   shapeName=thigh_l_capsule
   center=<-15.9432 1.2775 1.8553>
   rotation=<rot 90 180 -175.3141>
  }
  {shapeType=Capsule, length=12.375, radius=7.375, shapeName=thigh_l_capsule, center=<-23.9407 5.2037 1.1132>, rotation=<rot 90 0.0099 0.0099>}
  {shapeType=Capsule, length=5.6366, radius=6.0294, shapeName=thigh_l_capsule, center=<-4.4744 3.9728 0.1803>, rotation=<rot -44.9996 0 0>}
 ]
 thigh_r [{shapeType=Capsule, length=16.6581, radius=8.81, shapeName=thigh_r_capsule, center=<14.5658 0.13 -0.8394>, rotation=<rot 90 0 0>}]
}
constraints [
 {
  bone1Name=dyn_flap_front_MSTR
  bone2Name=pelvis
  swing1LimitDegrees=45
  swing1Motion=Free
  swing2LimitDegrees=45
  swing2Motion=Free
  twistLimitDegrees=45
  twistMotion=Free
 }
 {bone1Name=dyn_back_flap_02, bone2Name=dyn_back_flap_01, swing1LimitDegrees=15, swing2LimitDegrees=45, twistLimitDegrees=5}
 {bone1Name=dyn_back_flap_01, bone2Name=dyn_back_flap_MSTR, swing1LimitDegrees=15, swing2LimitDegrees=45, twistLimitDegrees=5}
 {
  bone1Name=dyn_back_flap_MSTR
  bone2Name=pelvis
  swing1LimitDegrees=45
  swing1Motion=Free
  swing2LimitDegrees=45
  swing2Motion=Free
  twistLimitDegrees=45
  twistMotion=Free
 }
 {
  bone1Name=thigh_r
  bone2Name=pelvis
  swing1LimitDegrees=45
  swing1Motion=Free
  swing2LimitDegrees=45
  swing2Motion=Free
  twistLimitDegrees=45
  twistMotion=Free
 }
 {
  bone1Name=dyn_strap_l_MSTR
  bone2Name=thigh_l
  swing1LimitDegrees=45
  swing1Motion=Free
  swing2LimitDegrees=45
  swing2Motion=Free
  twistLimitDegrees=45
  twistMotion=Free
 }
 {bone1Name=dyn_strap_l_01, bone2Name=dyn_strap_l_MSTR, swing1LimitDegrees=45, swing2LimitDegrees=60, twistLimitDegrees=5}
 {bone1Name=dyn_strap_l_02, bone2Name=dyn_strap_l_01, swing1LimitDegrees=45, swing2LimitDegrees=60, twistLimitDegrees=5}
 {bone1Name=dyn_strap_l_03, bone2Name=dyn_strap_l_02, swing1LimitDegrees=45, swing2LimitDegrees=60, twistLimitDegrees=5}
 {bone1Name=dyn_strap_l_04, bone2Name=dyn_strap_l_03, swing1LimitDegrees=45, swing2LimitDegrees=60, twistLimitDegrees=5}
 {
  bone1Name=dyn_strap_r_MSTR
  bone2Name=pelvis
  swing1LimitDegrees=45
  swing1Motion=Free
  swing2LimitDegrees=45
  swing2Motion=Free
  twistLimitDegrees=45
  twistMotion=Free
 }
 {bone1Name=dyn_strap_r_01, bone2Name=dyn_strap_r_MSTR, swing1LimitDegrees=45, swing2LimitDegrees=45, twistLimitDegrees=3}
 {bone1Name=dyn_strap_r_02, bone2Name=dyn_strap_r_01, swing1LimitDegrees=45, swing2LimitDegrees=45, twistLimitDegrees=3}
 {bone1Name=dyn_strap_r_03, bone2Name=dyn_strap_r_02, swing1LimitDegrees=45, swing2LimitDegrees=45, twistLimitDegrees=3}
 {bone1Name=dyn_strap_r_04, bone2Name=dyn_strap_r_03, swing1LimitDegrees=45, swing2LimitDegrees=45, twistLimitDegrees=3}
 {
  bone1Name=thigh_l
  bone2Name=pelvis
  swing1LimitDegrees=45
  swing1Motion=Free
  swing2LimitDegrees=45
  swing2Motion=Free
  twistLimitDegrees=45
  twistMotion=Free
 }
 {
  bone1Name=dyn_straps_driven_col
  bone2Name=pelvis
  swing1LimitDegrees=45
  swing1Motion=Free
  swing2LimitDegrees=45
  swing2Motion=Free
  twistLimitDegrees=45
  twistMotion=Free
 }
]
