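# Physics asset for the dangling grenades on the male medium Skull Punk player body's outfit.
# Two grenade capsules hang from tiny MSTR anchor bodies through constraints with tight twist and asymmetric swing limits.
# The grenade and front/back flap anchors, plus a left thigh capsule, join to the pelvis with angularly free constraints.
# Flaps have only anchor bodies here, with no simulated segments below them.
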PhysicsAsset /BRCosmetics/Characters/Player/Male/Medium/Bodies/M_MED_Skull_Punk/Meshes/M_MED_Skull_Punk_grenade_Physics
bodies {
 dyn_back_flap_MSTR [{shapeType=Capsule, length=0.505, radius=0.505, shapeName=dyn_back_flap_MSTR_capsule, rotation=<rot 0 -90 0>}]
 dyn_flap_front_MSTR [{shapeType=Capsule, length=0.505, radius=0.505, shapeName=dyn_flap_front_MSTR_capsule, rotation=<rot 0 -90 0>}]
 dyn_grenade_01 [{shapeType=Capsule, length=9.6401, radius=2.8784, shapeName=dyn_grenade_01_capsule, center=<4.5 0 -5>, rotation=<rot 13.9208 0 0>}]
 dyn_grenade_01_MSTR [{shapeType=Capsule, length=0.505, radius=0.505, shapeName=dyn_grenade_01_MSTR_capsule, rotation=<rot 0 -90 0>}]
 dyn_grenade_02 [{shapeType=Capsule, length=9.6605, radius=2.8584, shapeName=dyn_grenade_02_capsule, center=<4 0 -4>, rotation=<rot 9.9809 0 4.9999>}]
 dyn_grenade_02_MSTR [{shapeType=Capsule, length=0.505, radius=0.505, shapeName=dyn_grenade_02_MSTR_capsule, rotation=<rot 0 -90 0>}]
 pelvis [{shapeType=Capsule, length=8.1341, radius=11.0802}]
 thigh_twist_01_l [{
   shapeType=Capsule
   length=3.165
   radius=5.8559
   shapeName=thigh_twist_01_l_capsule
   center=<10.0672 1.5882 -2.0749>
   rotation=<rot 21.0138 -179.8254 -81.9629>
  }]
}
constraints [
 {
  bone1Name=dyn_flap_front_MSTR
  bone2Name=pelvis
  swing1LimitDegrees=45
  swing1Motion=Free
  swing2LimitDegrees=45
  swing2Motion=Free
  twistLimitDegrees=45
  twistMotion=Free
 }
 {
  bone1Name=dyn_back_flap_MSTR
  bone2Name=pelvis
  swing1LimitDegrees=45
  swing1Motion=Free
  swing2LimitDegrees=45
  swing2Motion=Free
  twistLimitDegrees=45
  twistMotion=Free
 }
 {bone1Name=dyn_grenade_02, bone2Name=dyn_grenade_02_MSTR, swing1LimitDegrees=35, swing2LimitDegrees=15, twistLimitDegrees=5}
 {bone1Name=dyn_grenade_01, bone2Name=dyn_grenade_01_MSTR, swing1LimitDegrees=35, swing2LimitDegrees=15, twistLimitDegrees=5}
 {
  bone1Name=dyn_grenade_01_MSTR
  bone2Name=pelvis
  swing1LimitDegrees=45
  swing1Motion=Free
  swing2LimitDegrees=45
  swing2Motion=Free
  twistLimitDegrees=45
  twistMotion=Free
 }
 {
  bone1Name=dyn_grenade_02_MSTR
  bone2Name=pelvis
  swing1LimitDegrees=45
  swing1Motion=Free
  swing2LimitDegrees=45
  swing2Motion=Free
  twistLimitDegrees=45
  twistMotion=Free
 }
 {
  bone1Name=thigh_twist_01_l
  bone2Name=pelvis
  swing1LimitDegrees=45
  swing1Motion=Free
  swing2LimitDegrees=45
  swing2Motion=Free
  twistLimitDegrees=45
  twistMotion=Free
 }
]
