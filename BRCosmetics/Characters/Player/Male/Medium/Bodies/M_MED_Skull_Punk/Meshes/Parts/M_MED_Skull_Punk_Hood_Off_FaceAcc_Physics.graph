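# Physics asset for the hood-off face accessory part of the male medium Skull Punk character, setting up physics for its lowered hood.
# Five hood bones (back, back-left, back-right, left, right) carry flat box bodies, each constrained to its own tiny MSTR box body.
# The joints allow moderate swing with very little twist, and no constraints link the hood segments to each other.
# A capsule on the head, also unconstrained, provides a head collision shape.
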PhysicsAsset /BRCosmetics/Characters/Player/Male/Medium/Bodies/M_MED_Skull_Punk/Meshes/Parts/M_MED_Skull_Punk_Hood_Off_FaceAcc_Physics
bodies {
 dyn_hood_back [{shapeType=Box, extentX=10, extentY=15, extentZ=3, shapeName=dyn_hood_back_box, center=<2 0 0>}]
 dyn_hood_back_MSTR [{shapeType=Box, extentX=1.01, extentY=1.01, extentZ=1.01, shapeName=dyn_hood_back_MSTR_box, rotation=<rot 0 -90 0>}]
 dyn_hood_bk_l [{shapeType=Box, extentX=10, extentY=15, extentZ=3, shapeName=dyn_hood_bk_l_box, center=<2 0 0>, rotation=<rot 13.2644 8.6812 33.6417>}]
 dyn_hood_bk_l_MSTR [{shapeType=Box, extentX=1.01, extentY=1.01, extentZ=1.01, shapeName=dyn_hood_bk_l_MSTR_box, rotation=<rot 0 -90 0>}]
 dyn_hood_bk_r [{shapeType=Box, extentX=10, extentY=15, extentZ=3, shapeName=dyn_hood_bk_r_box, center=<2 0 0>}]
 dyn_hood_bk_r_MSTR [{shapeType=Box, extentX=1.01, extentY=1.01, extentZ=1.01, shapeName=dyn_hood_bk_r_MSTR_box, rotation=<rot 0 -90 0>}]
 dyn_hood_l [{shapeType=Box, extentX=10, extentY=15, extentZ=3, shapeName=dyn_hood_l_box, center=<2 0 0>, rotation=<rot 0 0 9.8236>}]
 dyn_hood_l_MSTR [{shapeType=Box, extentX=1.01, extentY=1.01, extentZ=1.01, shapeName=dyn_hood_l_MSTR_box, rotation=<rot 0 -90 0>}]
 dyn_hood_r [{shapeType=Box, extentX=10, extentY=15, extentZ=3, shapeName=dyn_hood_r_box, center=<2 0 0>}]
 dyn_hood_r_MSTR [{shapeType=Box, extentX=1.01, extentY=1.01, extentZ=1.01, shapeName=dyn_hood_r_MSTR_box, rotation=<rot 0 -90 0>}]
 head [{shapeType=Capsule, length=9.6026, radius=7.5154, center=<5.2784 3.4926 -0.0041>, rotation=<rot 0.0246 36.4371 -89.938>}]
}
constraints [
 {bone1Name=dyn_hood_back, bone2Name=dyn_hood_back_MSTR, swing1LimitDegrees=45, swing2LimitDegrees=45, twistLimitDegrees=3}
 {bone1Name=dyn_hood_bk_l, bone2Name=dyn_hood_bk_l_MSTR, swing1LimitDegrees=45, swing2LimitDegrees=45, twistLimitDegrees=3}
 {bone1Name=dyn_hood_bk_r, bone2Name=dyn_hood_bk_r_MSTR, swing1LimitDegrees=45, swing2LimitDegrees=45, twistLimitDegrees=3}
 {bone1Name=dyn_hood_l, bone2Name=dyn_hood_l_MSTR, swing1LimitDegrees=45, swing2LimitDegrees=45, twistLimitDegrees=3}
 {bone1Name=dyn_hood_r, bone2Name=dyn_hood_r_MSTR, swing1LimitDegrees=45, swing2LimitDegrees=45, twistLimitDegrees=3}
]
