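# Post-process style AnimBP for the Skull Punk hood-off face accessory mesh, copying the attached parent's pose.
# While crouching, head rotation relative to spine_05 drives additive offsets on two rear hood bones, using separate moving and stationary alpha ranges blended by bIsMoving2D.
# A base-bone-space RigidBody node then simulates the accessory physics, and rotation constraints make dyn_eye bones follow L_eye and R_eye.
# A second RigidBody node is left unconnected.
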AnimBlueprint /BRCosmetics/Characters/Player/Male/Medium/Bodies/M_MED_Skull_Punk/Meshes/Parts/M_MED_Skull_Punk_Hood_Off_FaceAcc_AnimBP
target_skeleton=&/M_MED_Skull_Punk_Hood_Off_FaceAcc_Skeleton
graphs [
 AnimGraph {nodes [
   RigidBody() {
    bClampLinearTranslationLimitToRefPose=true
    bOverrideWorldGravity=true
    overridePhysicsAsset=&/M_MED_Skull_Punk_Hood_Off_FaceAcc_Physics
    simulationSpace=BaseBoneSpace
    alphaBoolBlend {blendOption=Linear}
    baseBoneRef {boneName=spine_04}
    componentAppliedLinearAccClamp=<300 300 300>
    componentLinearAccScale=<0.3 0.3 0.3>
    componentLinearVelScale=<0.3 0.3 0.3>
    overrideWorldGravity {z=-980}
   }
   n1 = Get hood_0() -> (hood_0:double)
   n2 = "Transform (Modify) Bone - Bone: dyn_hood_back_MSTR"(
    ComponentPose=Local To Component(LocalPose=Copy Pose From Mesh() {bUseAttachedParent=true})
    Alpha=@n1
    Translation="-1.840,4.771,11.148"
    Rotation="30,0.000,0.000"
    Scale="1,1,1"
   ) {
    rotationMode=BMM_Additive
    rotationSpace=BCS_BoneSpace
    translationMode=BMM_Additive
    translationSpace=BCS_BoneSpace
    alphaBoolBlend {blendOption=Linear}
    boneToModify {boneName=dyn_hood_back_MSTR}
    rotation {pitch=30}
    translation=<-1.84 4.771 11.148>
   }
   n3 = "Transform (Modify) Bone - Bone: dyn_hood_bk_r_MSTR"(ComponentPose=@n2, Alpha=@n1, Translation="-6.132,-0.000,7.109", Rotation="45,0,0", Scale="1,1,1") {
    rotationMode=BMM_Additive
    rotationSpace=BCS_BoneSpace
    translationMode=BMM_Additive
    translationSpace=BCS_BoneSpace
    alphaBoolBlend {blendOption=Linear}
    boneToModify {boneName=dyn_hood_bk_r_MSTR}
    rotation {pitch=45}
    translation {x=-6.1325, z=7.1086}
   }
   n4 = Component To Local(ComponentPose=@n3)
   n5 = Get hood() -> (hood:double)
   n6 = "Transform (Modify) Bone - Bone: dyn_hood_back_MSTR"(
    ComponentPose=Local To Component(LocalPose=Copy Pose From Mesh() {bUseAttachedParent=true})
    Alpha=@n5
    Translation="-1.840,4.771,11.148"
    Rotation="30,0.000,0.000"
    Scale="1,1,1"
   ) {
    rotationMode=BMM_Additive
    rotationSpace=BCS_BoneSpace
    translationMode=BMM_Additive
    translationSpace=BCS_BoneSpace
    alphaBoolBlend {blendOption=Linear}
    boneToModify {boneName=dyn_hood_back_MSTR}
    rotation {pitch=30}
    translation=<-1.84 4.771 11.148>
   }
   n7 = "Transform (Modify) Bone - Bone: dyn_hood_bk_r_MSTR"(ComponentPose=@n6, Alpha=@n5, Translation="-6.132,-0.000,7.109", Rotation="45,0,0", Scale="1,1,1") {
    rotationMode=BMM_Additive
    rotationSpace=BCS_BoneSpace
    translationMode=BMM_Additive
    translationSpace=BCS_BoneSpace
    alphaBoolBlend {blendOption=Linear}
    boneToModify {boneName=dyn_hood_bk_r_MSTR}
    rotation {pitch=45}
    translation=<-6.1325 0 7.1086>
   }
   n8 = Component To Local(ComponentPose=@n7)
   n9 = Get bIsMoving2D()
   n10 = RigidBody!(
    ComponentPose=Local To Component(LocalPose=Blend Poses by bool(BlendPose_0=@n8, BlendPose_1=@n4, bActiveValue=@n9) {blendType=Linear})
    bAlphaBoolEnabled:bool
    OverrideWorldGravity:Vector="0,0,-980"
    ExternalForce
   ) -> (Pose) {
    alphaInputType=Bool
    bClampLinearTranslationLimitToRefPose=true
    bOverrideWorldGravity=true
    overridePhysicsAsset=&M_MED_Skull_Punk_Hood_Off_FaceAcc_Physics
    simulationSpace=BaseBoneSpace
    alphaBoolBlend {blendInTime=0.1, blendOption=Linear, blendOutTime=0.1}
    baseBoneRef {boneName=spine_05}
    componentAppliedLinearAccClamp=<3000 3000 3000>
    componentLinearAccScale=<0.3 0.3 0.3>
    componentLinearVelScale=<0.3 0.3 0.3>
    overrideWorldGravity {z=-980}
    simSpaceSettings {
     maxAngularAcceleration=200
     maxAngularVelocity=5
     maxLinearAcceleration=1000
     maxLinearVelocity=300
     velocityScaleZ=0
     worldAlpha=1
     externalLinearDragV=<1 1>
    }
   }
   n11 = "Constraint - dyn_eye_l (R)"(ComponentPose=@n10, Alpha=1, ConstraintWeights_0=1) {
    alphaBoolBlend {blendOption=Linear}
    boneToModify {boneName=dyn_eye_l}
    constraintSetup [{transformType=Rotation, targetBone {boneName=L_eye}}]
    constraintWeights [1]
   }
   n12 = "Constraint - dyn_eye_r (R)"(ComponentPose=@n11, Alpha=1, ConstraintWeights_0=1) {
    alphaBoolBlend {blendOption=Linear}
    boneToModify {boneName=dyn_eye_r}
    constraintSetup [{transformType=Rotation, targetBone {boneName=R_eye}}]
    constraintWeights [1]
   }
   Output Pose(Result=Component To Local(ComponentPose=@n12)) {layerGroup=None}
  ]}
 EventGraph {nodes [
   Event BlueprintUpdateAnimation()
   n1 = Get bIsCrouching()
   > n2 = Branch(Condition=@n1)
   > n3 = Branch(Condition=AND Boolean(A=@n1, B=Get bIsMoving2D()))
   > Set hood(hood:double=GetAlphaFromReferenceTransformDelta(SkeletalMeshComponent=Get Skel_Mesh(), BoneName=head, BaseName=spine_05, InputRange=(X=-40,Y=-70), TwistAxis="1,0,0")) -> (Output_Get:double)
   Set hood_0(
    execute=@n3.else
    hood_0:double=GetAlphaFromReferenceTransformDelta(SkeletalMeshComponent=Get Skel_Mesh(), BoneName=head, BaseName=spine_05, InputRange=(X=-15,Y=-60), TwistAxis="1,0,0")
   ) -> (Output_Get:double)
   Set hood(execute=@n2.else, hood=0)
   > Set hood_0(hood_0=0)
   Event BlueprintInitializeAnimation()
   > Set Skel_Mesh(Skel_Mesh=GetOwnerSkeletalMeshComponent())
   Set hood(hood=0)
  ]}
]
variables [
 hood {type=number}
 hood_0 {type=number}
 hood_crouch {type=boolean}
 Skel_Mesh {type=SkeletalMeshComponent}
]
